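# Sound cue for the Figment S08 Athena rocket launcher fire shot, built as three distance layers.
# A mixer blends a close layer that randomly picks one of two launch waves, a medium shot, and a quieter distant shot.
# Each layer has its own attenuation and sound class override, with modulators on the close and distant layers and pitch variation only on the distant one.
# Local and global gunfire concurrency sets govern simultaneous playback.
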SoundCue /Figment_Launcher_Rocket_Athena/Sounds/S08/SC_Figment_S08_Weapon_LAU_Rocket_Launcher_Fire_Athena_Cue
root=Mixer_3
nodes [
 {
  id=Attenuation_10
  children [WavePlayer_42]
  properties {attenuationSettings=&Fort_Weapon_Projectile_Medium_Mono_Athena_Att}
 }
 {
  id=Attenuation_11
  children [Modulator_7]
  properties {attenuationSettings=&Fort_Weapon_Projectile_Distant_Mono_Athena_Att}
 }
 {id=Attenuation_9, children [Modulator_6], properties {attenuationSettings=&Fort_Weapon_Projectile_Close_Athena_Att}}
 {id=Mixer_3, children [SoundClass_1,SoundClass_2,SoundClass_6], properties {inputVolume [1,1,0.6]}}
 {id=Modulator_6, children [Random_3], properties {pitchMax=1, pitchMin=1}}
 {id=Modulator_7, children [WavePlayer_45], properties {pitchMax=1.1}}
 {id=Random_3, children [WavePlayer_18, WavePlayer_19], properties {bShouldExcludeFromBranchCulling=true, weights [1,1]}}
 {
  id=SoundClass_1
  children [Attenuation_9]
  properties {soundClassOverride=&Fort_WeaponRanged_Fire_3P_PSM}
 }
 {
  id=SoundClass_2
  children [Attenuation_10]
  properties {soundClassOverride=&Fort_WeaponRanged_Fire_Distant}
 }
 {
  id=SoundClass_6
  children [Attenuation_11]
  properties {soundClassOverride=&Fort_WeaponRanged_Fire_Distant}
 }
 {id=WavePlayer_18, properties {soundWaveAssetPtr=/Game/Sounds/Fort_Weapons/Launchers/rocket_launcher_start_02}}
 {id=WavePlayer_19, properties {soundWaveAssetPtr=/Game/Sounds/Fort_Weapons/Launchers/rocket_launcher_start_01}}
 {id=WavePlayer_42, properties {soundWaveAssetPtr=/Game/Sounds/Fort_Weapons/Launchers/Rocket_Launcher_Medium_ShotC01}}
 {id=WavePlayer_45, properties {soundWaveAssetPtr=/Game/Sounds/Fort_Weapons/Launchers/Rocket_Launcher_Distant_ShotE01}}
]
properties {
 maxDistance=27500
 soundClassObject=&Fort_WeaponRanged_Fire_3P_PSM
 volumeMultiplier=1.2
 attenuationOverrides {falloffDistance=5000, attenuationShapeExtents {x=1000}}
 concurrencyOverrides {bLimitToOwner=true, maxCount=3, resolutionRule=PreventNew}
 concurrencySet [&Weapon_Gunfire_Local_Conc, &Weapon_Gunfire_Global_Conc]
}
