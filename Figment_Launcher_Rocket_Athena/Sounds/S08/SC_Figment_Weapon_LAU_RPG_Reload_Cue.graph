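# Sound cue for the reload foley of the Figment RPG rocket launcher weapon.
# A single wave player feeds the shared Fort launcher reload sound through a Modulator node left at default settings.
# The cue boosts volume, routes to the weapon foley sound class and uses the shared weapon reload attenuation asset.
# Its concurrency override allows one instance at a time, stopping the farthest and then preventing new ones.
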SoundCue /Figment_Launcher_Rocket_Athena/Sounds/S08/SC_Figment_Weapon_LAU_RPG_Reload_Cue
root=Modulator_1
nodes [
 {id=Modulator_1, children [WavePlayer_3]}
 {id=WavePlayer_3, properties {soundWaveAssetPtr=/Game/Sounds/Fort_Weapons/Launchers/weapons_launchers_reload_01}}
]
properties {
 attenuationSettings=&Fort_Weapon_Reload_Att
 bOverrideConcurrency=true
 maxDistance=2250
 soundClassObject=&Fort_WeaponFoley
 volumeMultiplier=1.5
 attenuationOverrides {falloffDistance=5000, attenuationShapeExtents {x=1000}}
 concurrencyOverrides {maxCount=1, resolutionRule=StopFarthestThenPreventNew}
}
