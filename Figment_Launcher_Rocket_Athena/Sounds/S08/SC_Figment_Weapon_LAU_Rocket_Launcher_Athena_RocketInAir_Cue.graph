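# Sound cue for the in-flight loop of the Figment rocket launcher's projectile, attenuated with distant-rocket settings.
# A quality switch picks between two parallel chains: a looping wave for the first two levels, an "opt" wave not flagged as looping for the other two.
# Each chain runs its wave through a volume fade-in envelope, pitch modulation and smoothed, exaggerated Doppler.
# Per-owner concurrency caps simultaneous instances, stopping the quietest.
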SoundCue /Figment_Launcher_Rocket_Athena/Sounds/S08/SC_Figment_Weapon_LAU_Rocket_Launcher_Athena_RocketInAir_Cue
root=QualityLevel_0
nodes [
 {id=Doppler_0, children [Modulator_0], properties {bUseSmoothing=true, dopplerIntensity=3}}
 {id=Doppler_1, children [Modulator_7], properties {bUseSmoothing=true, dopplerIntensity=3}}
 {
  id=Enveloper_0
  children [WavePlayer_1]
  properties {
   pitchCurve {editorCurveData {keys [{time=0}]}}
   volumeCurve {editorCurveData {keys [{interpMode=RCIM_Cubic, tangentMode=RCTM_User, time=0, value=0}, {interpMode=RCIM_Cubic, tangentMode=RCTM_User, time=1, value=1}]}}
  }
 }
 {
  id=Enveloper_4
  children [WavePlayer_0]
  properties {
   pitchCurve {editorCurveData {keys [{time=0}]}}
   volumeCurve {editorCurveData {keys [{interpMode=RCIM_Cubic, tangentMode=RCTM_User, time=0, value=0}, {interpMode=RCIM_Cubic, tangentMode=RCTM_User, time=1, value=1}]}}
  }
 }
 {id=Modulator_0, children [Enveloper_0], properties {pitchMax=1.1}}
 {id=Modulator_7, children [Enveloper_4], properties {pitchMax=1.1}}
 {id=QualityLevel_0, children [Doppler_1,Doppler_1,Doppler_0,Doppler_0]}
 {id=WavePlayer_0, properties {bLooping=true, soundWaveAssetPtr=/Game/Sounds/Fort_Weapons/Launchers/RPG_InAir_NewLoop_LP2}}
 {id=WavePlayer_1, properties {soundWaveAssetPtr=/Game/Sounds/Fort_Weapons/Launchers/RPG_InAir_NewLoop_LP2_opt}}
]
properties {
 attenuationSettings=&/ATT_Figment_Weapon_Rocket_Distant
 bOverrideConcurrency=true
 maxDistance=15000
 soundClassObject=&Fort_WeaponRanged_Fire_3P_PSM
 volumeMultiplier=0.8
 attenuationOverrides {falloffDistance=5000, attenuationShapeExtents {x=1000}}
 concurrencyOverrides {bLimitToOwner=true, maxCount=3, resolutionRule=StopQuietest}
}
